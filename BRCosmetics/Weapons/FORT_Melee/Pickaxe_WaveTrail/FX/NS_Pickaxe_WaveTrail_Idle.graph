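# Niagara idle effect for the WaveTrail pickaxe cosmetic, showing a persistent cyan-blue glow or shape.
# One emitter, DirectionalBurst, spawns an instantaneous burst of particles that loop their lifetime instead of dying, until the emitter deactivates.
# Those particles draw through one sprite renderer and two mesh renderers.
# ScaleColor takes alpha from a front-end float modifier, making the effect fainter in the front end than in gameplay.
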NiagaraSystem /BRCosmetics/Weapons/FORT_Melee/Pickaxe_WaveTrail/FX/NS_Pickaxe_WaveTrail_Idle
dependencies {usedDataInterfaces [], usedDynamicInputs [], usedModules [&SystemState]}
emitters {DirectionalBurst {
  emitterUpdateScript [
   &EmitterState(Life Cycle Mode=::NewEnumerator1, Loop Behavior=::NewEnumerator1, Loop Duration Mode=::NewEnumerator1, UseLoopDelay=false)
   &SpawnBurst_Instantaneous(Use Spawn Probability=false, UseLoopCountLimit=false)
  ]
  particleSpawnScript [&InitializeParticle(
    Color=<rgba 0 0.7 1 1>
    Mass Max=2
    Mass Mode=::NewEnumerator3
    Mesh Scale Mode=::NewEnumerator3
    UsePositionOffset=false
    Write Material Random=true
   )]
  particleUpdateScript [
   &ParticleState(
    Kill Particles When Lifetime Has Elapsed=false
    Let Infinitely Lived Particles Die When Emitter Deactivates=true
    Loop Particles Lifetime=true
   )
   &ScaleColor(Scale Alpha=&FN_FrontEndFloatModifier(Gameplay=0.364, Front End=0.1), ScaleA=true, ScaleRGB=true)
  ]
  rendererClasses [&NiagaraSpriteRendererProperties, &NiagaraMeshRendererProperties]
  renderers [
   {rendererClass=&NiagaraSpriteRendererProperties}
   {rendererClass=&NiagaraMeshRendererProperties, rendererIndex=1}
   {rendererClass=&NiagaraMeshRendererProperties, rendererIndex=2}
  ]
 }}
summary {
 systemName=NS_Pickaxe_WaveTrail_Idle
 emitters [{
   emitterName=DirectionalBurst
   rendererClasses [&NiagaraSpriteRendererProperties, &NiagaraMeshRendererProperties]
  }]
 userVariables [
  User.bIsFrontEnd {type=NiagaraBool}
  User.bIsFrontEndPreview {type=NiagaraBool}
 ]
}
