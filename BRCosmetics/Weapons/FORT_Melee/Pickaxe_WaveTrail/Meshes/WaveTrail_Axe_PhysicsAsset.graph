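# Physics asset for the WaveTrail pickaxe, a two-handed melee weapon mesh in the BRCosmetics content.
# It holds a single body on the Melee_TwoHanded_Handle bone, built from one capsule offset along the handle and rotated from the bone's default orientation.
# It defines no constraints, so the axe has one simple collision shape with no articulated parts.
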PhysicsAsset /BRCosmetics/Weapons/FORT_Melee/Pickaxe_WaveTrail/Meshes/WaveTrail_Axe_PhysicsAsset
bodies {Melee_TwoHanded_Handle [{shapeType=Capsule, length=22.757, radius=6.1143, center=<-0.0311 -0.0027 9.9279>, rotation=<rot 0.1794 -90 -0.0281>}]}
constraints []
